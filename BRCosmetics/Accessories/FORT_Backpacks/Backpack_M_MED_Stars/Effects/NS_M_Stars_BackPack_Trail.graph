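# Niagara ribbon trail for the Stars backpack cosmetic, drawing seven translucent purple ribbons positioned from the user-supplied skeletal mesh.
# A system-level FN_Integrator value scales both spawn rate and opacity in gameplay.
# Front-end modifiers substitute a fixed rate and opacity scale, plus an added velocity to simulate a trail in menus.
# Ribbon opacity fades along a curve, and the material's pan speed differs by spawn group.
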NiagaraSystem /BRCosmetics/Accessories/FORT_Backpacks/Backpack_M_MED_Stars/Effects/NS_M_Stars_BackPack_Trail
dependencies {
 usedDataInterfaces [&NiagaraDataInterfaceCurve]
 usedDynamicInputs [
  &ClampFloat
  &Boolean_AndOperation
  &SetBoolByFloatComparison
  &VectorLength
 ]
 usedModules [&SystemState, &FN_Integrator]
}
emitters {Star_Trail {
  emitterUpdateScript [
   &EmitterState(Life Cycle Mode=::NewEnumerator1, Loop Duration=1, UseLoopDelay=false)
   /Game/Effects/Niagara/Modules/Ribbons/FN_SpawnRibbon_v3(
    Number of Ribbons=7
    Skeleton Mesh:NiagaraDataInterfaceSkeletalMesh=$User.SkeletalMesh
    SpawnRate:NiagaraFloat=&FN_FrontEndFloatModifier(Gameplay=&Lerp_Float(A=0, B=15, Alpha=$System.FN_Integrator.Value), Front End=15)
   )
  ]
  particleSpawnScript [
   &FN_SetSpawnGroup(Number of Spawn Groups=7)
   /Game/Effects/Niagara/Modules/Ribbons/FN_InitializeRibbon_v5(Enable Skeletal Position=true, SkeletalMesh:NiagaraDataInterfaceSkeletalMesh=$User.SkeletalMesh)
   &AddVelocity(Velocity=&FN_FrontEndVector3Modifier(Gameplay=<0 0 0>, Front End=<-60 -250 0>))
  ]
  particleUpdateScript [
   &SolveForcesAndVelocity(Clamp Velocity=false, Limit Acceleration=false)
   &ParticleState()
   /Game/Effects/Niagara/Modules/Ribbons/FN_UpdateRibbon_v5(Enable Skeletal Position=true, SkeletalMesh:NiagaraDataInterfaceSkeletalMesh=$User.SkeletalMesh)
   &FN_RibbonMaterial(Opacity Pan U Speed:NiagaraFloat=&MakeCustomFloatFromBool(Bool=HlslExpression() {hlslExpression="Particles.SpawnGroup >1"}, True Float=0.75, False Float=1.2))
   /Game/Effects/Niagara/Modules/Color/FN_ColorHueShift(Color:LinearColor=&MultiplyLinearColor(
     Linear Color=<rgba 0.5 0 1 0.08>
     Scale Factor=&MakeFloat4FromVectorAndFloat(
      "Vector (RGB)"=<1 1 1>
      "Float (Alpha)"=&FloatFromCurve(
       FloatCurve=DataInterface() {propertyValues {Curve {keys [{interpMode=RCIM_Cubic, tangentMode=RCTM_Break, time=0, value=1}, {interpMode=RCIM_Cubic, tangentMode=RCTM_Break, time=1, value=0}]}}}
       Scale Curve=&FN_FrontEndFloatModifier(Gameplay=&Lerp_Float(A=0, B=1, Alpha=$System.FN_Integrator.Value), Front End=1.15)
      )
     )
    ))
  ]
  rendererClasses [&NiagaraRibbonRendererProperties]
  renderers [{rendererClass=&NiagaraRibbonRendererProperties}]
 }}
summary {
 systemName=NS_M_Stars_BackPack_Trail
 emitters [{emitterName=Star_Trail, rendererClasses [&NiagaraRibbonRendererProperties]}]
 userVariables [
  User.bIsFrontEnd {type=NiagaraBool}
  User.bisFrontEndPreview {type=NiagaraBool}
  User.RibbonIDOffset {type=NiagaraInt32}
  User.SkeletalMesh {
   type=NiagaraDataInterfaceSkeletalMesh
   defaultValue {
    struct=NiagaraExt_VariableValue_DataInterface
    value {dataInterface=&self:NiagaraDataInterfaceSkeletalMesh_0}
   }
  }
  User.TrailOn {type=NiagaraFloat, defaultValue=15, description=""}
 ]
}
