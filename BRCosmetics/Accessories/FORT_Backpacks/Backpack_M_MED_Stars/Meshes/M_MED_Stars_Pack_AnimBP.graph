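# Anim Blueprint for the MED Stars backpack cosmetic that makes its cubes float, spin and sway with movement.
# From the ref pose it enlarges the pack bone, runs partially blended, spring-driven Anim Dynamics on six cube bones, then adds bone offsets.
# The update event builds two spin rotators and two sine/cosine bob vectors from real or game time.
# Cube_main only bobs, while cube_06 spins with a fixed offset.
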AnimBlueprint /BRCosmetics/Accessories/FORT_Backpacks/Backpack_M_MED_Stars/Meshes/M_MED_Stars_Pack_AnimBP
target_skeleton=&/M_MED_Stars_Pack_Skeleton
graphs [
 AnimGraph {nodes [
   n1 = "Transform (Modify) Bone - Bone: pack"(ComponentPose=Local To Component(LocalPose=Local Space Ref Pose()), Translation="0,0,0", Rotation="0,0,0", Scale="1.05,1.05,1.05") {scaleMode=BMM_Replace, alphaBoolBlend {blendOption=Linear}, boneToModify {boneName=pack}, scale=<1.05 1.05 1.05>}
   n2 = Anim Dynamics - Bone: dyn_cube_01(ComponentPose=@n1, Alpha=0.5, ExternalForce="0,0,0") {
    angularDampingOverride=0.999
    angularSpringConstant=150
    bAngularSpring=true
    bGravityOverrideInSimSpace=true
    bLinearSpring=true
    bOverrideAngularDamping=true
    bOverrideLinearDamping=true
    bUseGravityOverride=true
    gravityScale=0
    linearDampingOverride=0.995
    linearSpringConstant=50
    simulationSpace=Actor
    alphaBoolBlend {blendOption=Linear}
    alphaScaleBiasClamp {bInterpResult=true, interpSpeedDecreasing=3, interpSpeedIncreasing=5}
    boundBone {boneName=dyn_cube_01}
    componentAppliedLinearAccClamp=<6000 6000 6000>
    componentLinearAccScale=<0.05 0.05 0.05>
    componentLinearVelScale=<0.8 0.8 0.6>
    gravityOverride=<0.25 0.25 0.25>
    physicsBodyDefinitions [{
      sphereCollisionRadius=0
      boundBone {boneName=dyn_cube_01}
      boxExtents=<50 50 50>
      constraintSetup {angularLimitsMax=<0 0 0>, angularLimitsMin=<0 0 0>, angularTarget=<0 0 0>, linearAxesMax=<0.5 1 1>, linearAxesMin=<-0.5 -1 -1>}
      localJointOffset=<0 -3 0>
     }]
    relativeSpaceBone {boneName=root}
   }
   n3 = Anim Dynamics - Bone: dyn_cube_03(ComponentPose=@n2, Alpha=0.3, ExternalForce="0,0,0") {
    angularDampingOverride=0.999
    angularSpringConstant=150
    bAngularSpring=true
    bGravityOverrideInSimSpace=true
    bLinearSpring=true
    bOverrideAngularDamping=true
    bOverrideLinearDamping=true
    bUseGravityOverride=true
    gravityScale=0
    linearDampingOverride=0.995
    linearSpringConstant=50
    simulationSpace=Actor
    alphaBoolBlend {blendOption=Linear}
    alphaScaleBiasClamp {bInterpResult=true, interpSpeedDecreasing=3, interpSpeedIncreasing=5}
    boundBone {boneName=dyn_cube_03}
    componentAppliedLinearAccClamp=<6000 6000 6000>
    componentLinearAccScale=<0.1 0.1 0.3>
    componentLinearVelScale=<0.8 0.8 0.6>
    gravityOverride=<0.25 0.25 0.25>
    physicsBodyDefinitions [{
      sphereCollisionRadius=0
      boundBone {boneName=dyn_cube_03}
      boxExtents=<50 50 50>
      constraintSetup {angularLimitsMax=<0 0 0>, angularLimitsMin=<0 0 0>, angularTarget=<0 0 0>, linearAxesMax=<1 2 3>, linearAxesMin=<-1 0 -2>}
      localJointOffset=<0 -3 0>
     }]
    relativeSpaceBone {boneName=root}
   }
   n4 = Anim Dynamics - Bone: dyn_cube_05(ComponentPose=@n3, Alpha=0.2, ExternalForce="0,0,0") {
    angularDampingOverride=0.999
    angularSpringConstant=150
    bAngularSpring=true
    bGravityOverrideInSimSpace=true
    bLinearSpring=true
    bOverrideAngularDamping=true
    bOverrideLinearDamping=true
    bUseGravityOverride=true
    gravityScale=0
    linearDampingOverride=0.995
    linearSpringConstant=50
    simulationSpace=Actor
    alphaBoolBlend {blendOption=Linear}
    alphaScaleBiasClamp {bInterpResult=true, interpSpeedDecreasing=3, interpSpeedIncreasing=5}
    boundBone {boneName=dyn_cube_05}
    componentAppliedLinearAccClamp=<6000 6000 6000>
    componentLinearAccScale=<0.05 0.05 0.05>
    componentLinearVelScale=<0.8 1 0.6>
    gravityOverride=<0.25 0.25 0.25>
    physicsBodyDefinitions [{
      sphereCollisionRadius=0
      boundBone {boneName=dyn_cube_05}
      boxExtents=<50 50 50>
      constraintSetup {angularLimitsMax=<0 0 0>, angularLimitsMin=<0 0 0>, angularTarget=<0 0 0>, linearAxesMax=<1 2 1>, linearAxesMin=<-1 1 -2>}
      localJointOffset=<0 -3 0>
     }]
    relativeSpaceBone {boneName=root}
   }
   n5 = Anim Dynamics - Bone: dyn_cube_02(ComponentPose=@n4, Alpha=0.2, ExternalForce="0,0,0") {
    angularDampingOverride=0.999
    angularSpringConstant=150
    bAngularSpring=true
    bGravityOverrideInSimSpace=true
    bLinearSpring=true
    bOverrideAngularDamping=true
    bOverrideLinearDamping=true
    bUseGravityOverride=true
    gravityScale=0
    linearDampingOverride=0.995
    linearSpringConstant=50
    simulationSpace=Actor
    alphaBoolBlend {blendOption=Linear}
    alphaScaleBiasClamp {bInterpResult=true, interpSpeedDecreasing=3, interpSpeedIncreasing=5}
    boundBone {boneName=dyn_cube_02}
    componentAppliedLinearAccClamp=<6000 6000 6000>
    componentLinearAccScale=<0.1 0.1 0.3>
    componentLinearVelScale=<0.8 0.8 0.6>
    gravityOverride=<0.25 0.25 0.25>
    physicsBodyDefinitions [{
      sphereCollisionRadius=0
      boundBone {boneName=dyn_cube_02}
      boxExtents=<50 50 50>
      constraintSetup {angularLimitsMax=<0 0 0>, angularLimitsMin=<0 0 0>, angularTarget=<0 0 0>, linearAxesMax=<1 0 2>, linearAxesMin=<-1 0 -1>}
      localJointOffset=<0 -3 0>
     }]
    relativeSpaceBone {boneName=root}
   }
   n6 = Anim Dynamics - Bone: dyn_cube_04(ComponentPose=@n5, Alpha=0.5, ExternalForce="0,0,0") {
    angularDampingOverride=0.999
    angularSpringConstant=150
    bAngularSpring=true
    bGravityOverrideInSimSpace=true
    bLinearSpring=true
    bOverrideAngularDamping=true
    bOverrideLinearDamping=true
    bUseGravityOverride=true
    gravityScale=0
    linearDampingOverride=0.995
    linearSpringConstant=50
    simulationSpace=Actor
    alphaBoolBlend {blendOption=Linear}
    alphaScaleBiasClamp {bInterpResult=true, interpSpeedDecreasing=3, interpSpeedIncreasing=5}
    boundBone {boneName=dyn_cube_04}
    componentAppliedLinearAccClamp=<6000 6000 6000>
    componentLinearAccScale=<0.05 0.05 0.05>
    componentLinearVelScale=<0.8 0.8 0.6>
    gravityOverride=<0.25 0.25 0.25>
    physicsBodyDefinitions [{
      sphereCollisionRadius=0
      boundBone {boneName=dyn_cube_04}
      boxExtents=<50 50 50>
      constraintSetup {angularLimitsMax=<0 0 0>, angularLimitsMin=<0 0 0>, angularTarget=<0 0 0>, linearAxesMax=<1 1 1>, linearAxesMin=<-2 0 -2>}
      localJointOffset=<0 -3 0>
     }]
    relativeSpaceBone {boneName=root}
   }
   n7 = Anim Dynamics - Bone: dyn_cube_06(ComponentPose=@n6, Alpha=0.3, ExternalForce="0,0,0") {
    angularDampingOverride=0.999
    angularSpringConstant=150
    bAngularSpring=true
    bGravityOverrideInSimSpace=true
    bLinearSpring=true
    bOverrideAngularDamping=true
    bOverrideLinearDamping=true
    bUseGravityOverride=true
    gravityScale=0
    linearDampingOverride=0.995
    linearSpringConstant=50
    simulationSpace=Actor
    alphaBoolBlend {blendOption=Linear}
    alphaScaleBiasClamp {bInterpResult=true, interpSpeedDecreasing=3, interpSpeedIncreasing=5}
    boundBone {boneName=dyn_cube_06}
    componentAppliedLinearAccClamp=<6000 6000 6000>
    componentLinearAccScale=<0.1 0.1 0.3>
    componentLinearVelScale=<0.8 1 0.6>
    gravityOverride=<0.25 0.25 0.25>
    physicsBodyDefinitions [{
      sphereCollisionRadius=0
      boundBone {boneName=dyn_cube_06}
      boxExtents=<50 50 50>
      constraintSetup {angularLimitsMax=<0 0 0>, angularLimitsMin=<0 0 0>, angularTarget=<0 0 0>, linearAxesMax=<1 2 3>, linearAxesMin=<-0.5 1 0>}
      localJointOffset=<0 -3 0>
     }]
    relativeSpaceBone {boneName=root}
   }
   n8 = Get Cube_Float_A() -> (Cube_Float_A:Vector)
   n9 = "Transform (Modify) Bone - Bone: dyn_cube_main"(ComponentPose=@n7, Alpha=1, Translation=@n8, Rotation="0,0,0", Scale="1,1,1") {
    rotationMode=BMM_Additive
    rotationSpace=BCS_BoneSpace
    translationMode=BMM_Additive
    alphaBoolBlend {blendOption=Linear}
    alphaScaleBiasClamp {bInterpResult=true, interpSpeedDecreasing=4, interpSpeedIncreasing=6}
    boneToModify {boneName=dyn_cube_main}
   }
   n10 = Get Spinning_Rotator_02() -> (Spinning_Rotator_02:Rotator)
   n11 = "Transform (Modify) Bone - Bone: dyn_cube_01"(ComponentPose=@n9, Alpha=1, Translation=@n8, Rotation=@n10, Scale="1,1,1") {
    rotationMode=BMM_Additive
    rotationSpace=BCS_BoneSpace
    translationMode=BMM_Additive
    alphaBoolBlend {blendOption=Linear}
    alphaScaleBiasClamp {bInterpResult=true, interpSpeedDecreasing=4, interpSpeedIncreasing=6}
    boneToModify {boneName=dyn_cube_01}
   }
   n12 = Get Spinning_Rotator_01() -> (Spinning_Rotator_01:Rotator)
   n13 = "Transform (Modify) Bone - Bone: dyn_cube_02"(ComponentPose=@n11, Alpha=1, Translation=@n8, Rotation=@n12, Scale="1,1,1") {
    rotationMode=BMM_Additive
    rotationSpace=BCS_BoneSpace
    translationMode=BMM_Additive
    alphaBoolBlend {blendOption=Linear}
    alphaScaleBiasClamp {bInterpResult=true, interpSpeedDecreasing=4, interpSpeedIncreasing=6}
    boneToModify {boneName=dyn_cube_02}
   }
   n14 = Get Cube_Float_B() -> (Cube_Float_B:Vector)
   n15 = "Transform (Modify) Bone - Bone: dyn_cube_03"(ComponentPose=@n13, Alpha=1, Translation=@n14, Rotation=@n10, Scale="1,1,1") {
    rotationMode=BMM_Additive
    rotationSpace=BCS_BoneSpace
    translationMode=BMM_Additive
    alphaBoolBlend {blendOption=Linear}
    alphaScaleBiasClamp {bInterpResult=true, interpSpeedDecreasing=4, interpSpeedIncreasing=6}
    boneToModify {boneName=dyn_cube_03}
   }
   n16 = "Transform (Modify) Bone - Bone: dyn_cube_04"(ComponentPose=@n15, Alpha=1, Translation=@n8, Rotation=@n12, Scale="1,1,1") {
    rotationMode=BMM_Additive
    rotationSpace=BCS_BoneSpace
    translationMode=BMM_Additive
    alphaBoolBlend {blendOption=Linear}
    alphaScaleBiasClamp {bInterpResult=true, interpSpeedDecreasing=4, interpSpeedIncreasing=6}
    boneToModify {boneName=dyn_cube_04}
   }
   n17 = "Transform (Modify) Bone - Bone: dyn_cube_05"(ComponentPose=@n16, Alpha=1, Translation=@n14, Rotation=@n10, Scale="1,1,1") {
    rotationMode=BMM_Additive
    rotationSpace=BCS_BoneSpace
    translationMode=BMM_Additive
    alphaBoolBlend {blendOption=Linear}
    alphaScaleBiasClamp {bInterpResult=true, interpSpeedDecreasing=4, interpSpeedIncreasing=6}
    boneToModify {boneName=dyn_cube_05}
   }
   n18 = "Transform (Modify) Bone - Bone: dyn_cube_06"(ComponentPose=@n17, Alpha=1, Translation="0.000,0,-1.25", Rotation=@n12, Scale="1,1,1") {
    rotationMode=BMM_Additive
    rotationSpace=BCS_BoneSpace
    translationMode=BMM_Additive
    alphaBoolBlend {blendOption=Linear}
    alphaScaleBiasClamp {bInterpResult=true, interpSpeedDecreasing=4, interpSpeedIncreasing=6}
    boneToModify {boneName=dyn_cube_06}
    translation {z=-1.25}
   }
   Output Pose(Result=Component To Local(ComponentPose=@n18)) {layerGroup=None}
  ]}
 EventGraph {nodes [
   Event BlueprintUpdateAnimation()
   > n1 = Sequence()
   n2 = float * float(A=float / float(A=GetRealTimeSeconds(), B=-720) {is=title K2Node_CallFunction}, B=15000)
   > Set Spinning_Rotator_01(Spinning_Rotator_01:Rotator=MakeRotator(Roll=@n2, Yaw=@n2)) -> (Output_Get:Rotator)
   n3 = float / float(A=GetRealTimeSeconds(), B=720)
   > Set Spinning_Rotator_02(Spinning_Rotator_02:Rotator=MakeRotator(Roll=float * float(A=@n3, B=30000))) -> (Output_Get:Rotator)
   n4 = GetRealTimeSeconds()
   n5 = "Sin (Radians)"(A=float * float(A=float * float(A=1, B=1), B=@n4))
   n6 = "Sin (Radians)"(A=float * float(A=float * float(A=1, B=1.4), B=@n4))
   n7 = "Sin (Radians)"(A=float * float(A=float * float(A=1, B=1.2), B=@n4))
   Set Cube_Float_A(
    execute=@n1.then_1
    Cube_Float_A:Vector=MakeVector(X=float * float(A=@n7, B=0.3), Y=float * float(A=@n5, B=0.56), Z=float * float(A=@n6, B=0.22))
   ) -> (Output_Get:Vector)
   n8 = GetGameTimeInSeconds()
   n9 = "Cos (Radians)"!(A:double=float * float(A=float * float(A=1, B=1.5), B=@n8)) -> (ReturnValue:double)
   n10 = "Cos (Radians)"(A=float * float(A=float * float(A=1, B=1.15), B=@n8))
   n11 = "Cos (Radians)"(A=float * float(A=float * float(A=1, B=1.15), B=@n8))
   > Set Cube_Float_B(Cube_Float_B:Vector=MakeVector(X=float * float(A=@n9, B=0.2), Y=float * float(A=@n10, B=0.725), Z=float * float(A=@n11, B=0.3))) -> (Output_Get:Vector)
   TryGetPawnOwner()
  ]}
]
variables [
 Cube_Float_A {type=Vector}
 Cube_Float_B {type=Vector}
 Spinning_Rotator_01 {type=Rotator}
 Spinning_Rotator_02 {type=Rotator}
]
